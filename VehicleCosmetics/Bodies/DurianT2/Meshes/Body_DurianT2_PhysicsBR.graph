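# Physics asset for the DurianT2 vehicle cosmetic body, named as its BR variant.
# It holds a single body on the root bone and no constraints, so it defines no jointed or ragdoll parts.
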PhysicsAsset /VehicleCosmetics/Bodies/DurianT2/Meshes/Body_DurianT2_PhysicsBR
bodies {root []}
constraints []
